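# Physics asset for the DeLasMuertos medium male FaceAcc_01 mesh, defining dynamic hair bodies around the head and shoulders.
# Seven two- or three-bone chains of boxes, some capsule-tipped, hang from the head through angle-limited constraints.
# Right side and back-right chains swing wider than the others, while the back-left root is nearly rigid.
# Spheres and capsules cover head, neck, upper spine and clavicles, linked by free angular constraints.
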PhysicsAsset /BRCosmetics/Characters/Player/Male/Medium/Bodies/M_MED_DeLasMuertos/Meshes/M_MED_DeLasMuertos_FaceAcc_01_Dyn
bodies {
 clavicle_l [{shapeType=Capsule, length=10.5374, radius=9.6414, center=<10.2043 -1.2911 -0.305>, rotation=<rot 90 0 0>}]
 clavicle_r [{shapeType=Capsule, length=10.4349, radius=9.7438, center=<-10.2826 1.2871 0.8621>, rotation=<rot 90 0 -179.9999>}]
 dyn_hair_bk_01 [{shapeType=Box, extentX=6.3148, extentY=0.5, extentZ=7.5, center=<3.4005 0 -0.0002>}]
 dyn_hair_bk_02 [{shapeType=Box, extentX=5.0155, extentY=0.5, extentZ=7.5, center=<2.6124 0 -0.0001>}]
 dyn_hair_bk_03 [{shapeType=Box, extentX=5.3603, extentY=0.5, extentZ=7.5, center=<2.6868 -0.1805 -0.0755>, rotation=<rot 0 -10.0014 0>}]
 dyn_hair_bk_l_01 [{shapeType=Box, extentX=7.5, extentY=0.5, extentZ=7.5, center=<3.1627 0 -0.0001>}]
 dyn_hair_bk_l_02 [{shapeType=Box, extentX=5.2898, extentY=0.5, extentZ=7.5, center=<3.1627 0 -0.0001>}]
 dyn_hair_bk_l_03 [{shapeType=Capsule, length=6.3207, radius=0.9742, center=<0.5535 0.0001 0>}]
 dyn_hair_bk_r_01 [{shapeType=Box, extentX=7.5, extentY=0.5, extentZ=9.8635, center=<-3.1627 0 0>}]
 dyn_hair_bk_r_02 [{shapeType=Box, extentX=4.0481, extentY=0.5, extentZ=9.8635, center=<-2.1182 0.0001 0>}]
 dyn_hair_bk_r_03 [{shapeType=Capsule, length=17.5005, radius=1.5746, center=<-1.9145 0 0>, rotation=<rot 0 0 -18.1728>}]
 dyn_hair_fr_l_01 [{shapeType=Box, extentX=3.1554, extentY=1.8245, extentZ=2.6624, center=<1.2327 0 -0.0001>}]
 dyn_hair_fr_l_02 [{shapeType=Box, extentX=3.1554, extentY=1.8245, extentZ=2.6624, center=<1.4171 0 -0.0001>}]
 dyn_hair_fr_r_01 [{shapeType=Box, extentX=2.5199, extentY=1.8245, extentZ=2.3115, center=<-1.3894 0 0.0004>}]
 dyn_hair_fr_r_02 [{shapeType=Box, extentX=2.5199, extentY=1.8245, extentZ=2.3115, center=<-1.3894 0 0.0004>}]
 dyn_hair_l_01 [{shapeType=Box, extentX=4.8098, extentY=0.5, extentZ=7.5, center=<2.6918 0 -0.0001>}]
 dyn_hair_l_02 [{shapeType=Box, extentX=4.4187, extentY=0.5, extentZ=7.5, center=<2.4511 0.0001 -0.0001>}]
 dyn_hair_l_03 [{shapeType=Capsule, length=6.3207, radius=0.9742, center=<0.7819 0 0.0001>}]
 dyn_hair_r_01 [{shapeType=Box, extentX=4.8982, extentY=0.8575, extentZ=12.4486, center=<-2.5648 0.0001 0>}]
 dyn_hair_r_02 [{shapeType=Box, extentX=3.3856, extentY=1.6701, extentZ=12.4486, center=<-2.3546 0 -0.0001>}]
 dyn_hair_r_03 [{shapeType=Capsule, length=28.6197, radius=1.5746, center=<-1.2283 -0.0001 0>, rotation=<rot 0 0 -25.7237>}]
 head [
  {shapeType=Sphere, radius=6.412, center=<5.2907 -0.9379 0>}
  {shapeType=Capsule, length=6.1855, radius=5.6734, center=<2.1415 0 0>, rotation=<rot 90 0 0>}
 ]
 neck_01 [{shapeType=Capsule, length=6.0021, radius=7.6411, center=<3.1397 -1.121 -0.0001>, rotation=<rot 90 6.5366 0.0819>}]
 spine_05 [
  {shapeType=Capsule, length=5.7867, radius=13.655, center=<1.6485 -0.0361 -0.0287>}
  {shapeType=Capsule, length=3.9424, radius=9.0627, center=<10.3759 0.633 0>}
 ]
}
constraints [
 {bone1Name=dyn_hair_bk_l_01, bone2Name=head, swing1LimitDegrees=1, swing2LimitDegrees=1, twistLimitDegrees=1}
 {bone1Name=dyn_hair_bk_l_02, bone2Name=dyn_hair_bk_l_01, swing1LimitDegrees=20, swing2LimitDegrees=20, twistLimitDegrees=5}
 {bone1Name=dyn_hair_bk_l_03, bone2Name=dyn_hair_bk_l_02, swing1LimitDegrees=20, swing2LimitDegrees=20, twistLimitDegrees=5}
 {bone1Name=dyn_hair_bk_01, bone2Name=head, swing1LimitDegrees=20, swing2LimitDegrees=20, twistLimitDegrees=5}
 {bone1Name=dyn_hair_bk_02, bone2Name=dyn_hair_bk_01, swing1LimitDegrees=20, swing2LimitDegrees=20, twistLimitDegrees=5}
 {bone1Name=dyn_hair_bk_03, bone2Name=dyn_hair_bk_02, swing1LimitDegrees=20, swing2LimitDegrees=20, twistLimitDegrees=5}
 {bone1Name=dyn_hair_l_01, bone2Name=head, swing1LimitDegrees=20, swing2LimitDegrees=20, twistLimitDegrees=5}
 {bone1Name=dyn_hair_l_02, bone2Name=dyn_hair_l_01, swing1LimitDegrees=20, swing2LimitDegrees=20, twistLimitDegrees=5}
 {bone1Name=dyn_hair_l_03, bone2Name=dyn_hair_l_02, swing1LimitDegrees=20, swing2LimitDegrees=20, twistLimitDegrees=5}
 {bone1Name=dyn_hair_fr_l_01, bone2Name=head, swing1LimitDegrees=20, swing2LimitDegrees=20, twistLimitDegrees=5}
 {bone1Name=dyn_hair_fr_l_02, bone2Name=dyn_hair_fr_l_01, swing1LimitDegrees=20, swing2LimitDegrees=20, twistLimitDegrees=5}
 {bone1Name=dyn_hair_bk_r_01, bone2Name=head, swing1LimitDegrees=45, swing2LimitDegrees=45, twistLimitDegrees=5}
 {bone1Name=dyn_hair_bk_r_02, bone2Name=dyn_hair_bk_r_01, swing1LimitDegrees=45, swing2LimitDegrees=45, twistLimitDegrees=5}
 {bone1Name=dyn_hair_bk_r_03, bone2Name=dyn_hair_bk_r_02, swing1LimitDegrees=45, swing2LimitDegrees=45, twistLimitDegrees=5}
 {bone1Name=dyn_hair_r_01, bone2Name=head, swing1LimitDegrees=45, swing2LimitDegrees=45, twistLimitDegrees=5}
 {bone1Name=dyn_hair_r_02, bone2Name=dyn_hair_r_01, swing1LimitDegrees=45, swing2LimitDegrees=45, twistLimitDegrees=5}
 {bone1Name=dyn_hair_r_03, bone2Name=dyn_hair_r_02, swing1LimitDegrees=45, swing2LimitDegrees=45, twistLimitDegrees=5}
 {bone1Name=dyn_hair_fr_r_01, bone2Name=head, swing1LimitDegrees=20, swing2LimitDegrees=20, twistLimitDegrees=5}
 {bone1Name=dyn_hair_fr_r_02, bone2Name=dyn_hair_fr_r_01, swing1LimitDegrees=20, swing2LimitDegrees=20, twistLimitDegrees=5}
 {
  bone1Name=head
  bone2Name=neck_01
  swing1LimitDegrees=45
  swing1Motion=Free
  swing2LimitDegrees=45
  swing2Motion=Free
  twistLimitDegrees=45
  twistMotion=Free
 }
 {
  bone1Name=clavicle_l
  bone2Name=spine_05
  swing1LimitDegrees=45
  swing1Motion=Free
  swing2LimitDegrees=45
  swing2Motion=Free
  twistLimitDegrees=45
  twistMotion=Free
 }
 {
  bone1Name=neck_01
  bone2Name=spine_05
  swing1LimitDegrees=45
  swing1Motion=Free
  swing2LimitDegrees=45
  swing2Motion=Free
  twistLimitDegrees=45
  twistMotion=Free
 }
 {
  bone1Name=clavicle_r
  bone2Name=spine_05
  swing1LimitDegrees=45
  swing1Motion=Free
  swing2LimitDegrees=45
  swing2Motion=Free
  twistLimitDegrees=45
  twistMotion=Free
 }
]
