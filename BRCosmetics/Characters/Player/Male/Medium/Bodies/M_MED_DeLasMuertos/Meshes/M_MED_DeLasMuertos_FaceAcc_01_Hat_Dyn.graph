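# Physics asset for the dynamic hat face accessory of the male medium DeLasMuertos player cosmetic, setting up physics for its flower and hat side bones.
# Small capsules on dyn_flower, dyn_hat_l and dyn_hat_r accompany a larger capsule on the head bone.
# Each dyn bone is constrained directly to the head rather than chained, with identical moderate swing limits and a tight twist limit.
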PhysicsAsset /BRCosmetics/Characters/Player/Male/Medium/Bodies/M_MED_DeLasMuertos/Meshes/M_MED_DeLasMuertos_FaceAcc_01_Hat_Dyn
bodies {
 dyn_flower [{shapeType=Capsule, length=2, radius=2, center=<-2.7681 0.0001 0.1>, rotation=<rot 90 -0.0099 -0.0099>}]
 dyn_hat_l [{shapeType=Capsule, length=2, radius=2, center=<-2.6274 0.0002 -0.1>, rotation=<rot 90 -0.0099 -0.0099>}]
 dyn_hat_r [{shapeType=Capsule, length=2, radius=2, center=<2.7681 -0.0001 0.1>, rotation=<rot 90 -0.0198 -0.0198>}]
 head [{shapeType=Capsule, length=6.1855, radius=5.6734, center=<2.1415 0 0>, rotation=<rot 90 0 0>}]
}
constraints [
 {bone1Name=dyn_hat_l, bone2Name=head, swing1LimitDegrees=45, swing2LimitDegrees=45, twistLimitDegrees=5}
 {bone1Name=dyn_flower, bone2Name=head, swing1LimitDegrees=45, swing2LimitDegrees=45, twistLimitDegrees=5}
 {bone1Name=dyn_hat_r, bone2Name=head, swing1LimitDegrees=45, swing2LimitDegrees=45, twistLimitDegrees=5}
]
