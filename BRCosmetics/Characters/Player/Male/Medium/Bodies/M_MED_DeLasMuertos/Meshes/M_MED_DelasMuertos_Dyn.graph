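# Physics asset for the DeLasMuertos male medium player body cosmetic, covering its dangling tassel bones with capsule bodies.
# A three-link tassel chain hangs from each front shoulder pad, and single front and back tassels attach to each calf.
# Tassel joints carry swing limits and a much tighter twist limit.
# Larger upper-spine and upper-arm capsules form the torso volume, with upper arms and shoulder pads joined to the spine by angularly free constraints.
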PhysicsAsset /BRCosmetics/Characters/Player/Male/Medium/Bodies/M_MED_DeLasMuertos/Meshes/M_MED_DelasMuertos_Dyn
bodies {
 calf_l [{shapeType=Capsule, length=2, radius=2, rotation=<rot 90 0 0>}]
 calf_r [{shapeType=Capsule, length=2, radius=2, rotation=<rot 90 0 0>}]
 dyn_shoulderPad_fr_l [{shapeType=Capsule, length=2, radius=2}]
 dyn_shoulderPad_fr_r [{shapeType=Capsule, length=2, radius=2}]
 dyn_tassle_b_l [{shapeType=Capsule, length=2, radius=2, center=<0 0 -2.8989>}]
 dyn_tassle_b_r [{shapeType=Capsule, length=2, radius=2, center=<0 0 -2.8989>}]
 dyn_tassle_f_l [{shapeType=Capsule, length=2, radius=2, center=<0 0 -2.8989>}]
 dyn_tassle_f_r [{shapeType=Capsule, length=2, radius=2, center=<0 0 -2.8989>}]
 dyn_tassle_l_01 [{shapeType=Capsule, length=1.8319, radius=0.9276, center=<0 0 -1.3302>}]
 dyn_tassle_l_02 [{shapeType=Capsule, length=1.8319, radius=0.9276, center=<0 0 -1.3302>}]
 dyn_tassle_l_03 [{shapeType=Capsule, length=3.8216, radius=1.6745, center=<0.7532 1.0923 -3.3873>, rotation=<rot 9.6557 -2.6137 -15.2266>}]
 dyn_tassle_r_01 [{shapeType=Capsule, length=1.8319, radius=0.9276, center=<0 0 -1.3302>}]
 dyn_tassle_r_02 [{shapeType=Capsule, length=1.8319, radius=0.9276, center=<0 0 -1.3302>}]
 dyn_tassle_r_03 [{shapeType=Capsule, length=3.8216, radius=1.6745, center=<-0.7145 0.9444 -3.5473>, rotation=<rot -9.6558 2.6133 -15.2235>}]
 spine_05 [{shapeType=Capsule, length=10.4898, radius=10.379, center=<-7.1927 4.7002 0>}]
 upperarm_l [{shapeType=Capsule, length=18.1074, radius=5.7934, center=<11.5149 1.618 -0.0001>, rotation=<rot 90 24.6235 33.3785>}]
 upperarm_r [{shapeType=Capsule, length=18.1074, radius=5.7934, center=<-11.5149 -1.618 -0.0001>, rotation=<rot 90 11.7683 -159.4767>}]
}
constraints [
 {bone1Name=dyn_tassle_l_02, bone2Name=dyn_tassle_l_01, swing1LimitDegrees=45, swing2LimitDegrees=45, twistLimitDegrees=5}
 {bone1Name=dyn_tassle_l_03, bone2Name=dyn_tassle_l_02, swing1LimitDegrees=45, swing2LimitDegrees=45, twistLimitDegrees=5}
 {bone1Name=dyn_tassle_l_01, bone2Name=dyn_shoulderPad_fr_l, swing1LimitDegrees=45, swing2LimitDegrees=45, twistLimitDegrees=5}
 {bone1Name=dyn_tassle_r_02, bone2Name=dyn_tassle_r_01, swing1LimitDegrees=45, swing2LimitDegrees=45, twistLimitDegrees=5}
 {bone1Name=dyn_tassle_r_01, bone2Name=dyn_shoulderPad_fr_r, swing1LimitDegrees=45, swing2LimitDegrees=45, twistLimitDegrees=5}
 {bone1Name=dyn_tassle_r_03, bone2Name=dyn_tassle_r_02, swing1LimitDegrees=45, swing2LimitDegrees=45, twistLimitDegrees=5}
 {bone1Name=dyn_tassle_b_r, bone2Name=calf_r, swing1LimitDegrees=45, swing2LimitDegrees=45, twistLimitDegrees=5}
 {bone1Name=dyn_tassle_f_r, bone2Name=calf_r, swing1LimitDegrees=45, swing2LimitDegrees=45, twistLimitDegrees=5}
 {bone1Name=dyn_tassle_b_l, bone2Name=calf_l, swing1LimitDegrees=45, swing2LimitDegrees=45, twistLimitDegrees=5}
 {bone1Name=dyn_tassle_f_l, bone2Name=calf_l, swing1LimitDegrees=45, swing2LimitDegrees=45, twistLimitDegrees=5}
 {
  bone1Name=upperarm_l
  bone2Name=spine_05
  swing1LimitDegrees=45
  swing1Motion=Free
  swing2LimitDegrees=45
  swing2Motion=Free
  twistLimitDegrees=45
  twistMotion=Free
 }
 {
  bone1Name=dyn_shoulderPad_fr_l
  bone2Name=spine_05
  swing1LimitDegrees=45
  swing1Motion=Free
  swing2LimitDegrees=45
  swing2Motion=Free
  twistLimitDegrees=45
  twistMotion=Free
 }
 {
  bone1Name=upperarm_r
  bone2Name=spine_05
  swing1LimitDegrees=45
  swing1Motion=Free
  swing2LimitDegrees=45
  swing2Motion=Free
  twistLimitDegrees=45
  twistMotion=Free
 }
 {
  bone1Name=dyn_shoulderPad_fr_r
  bone2Name=spine_05
  swing1LimitDegrees=45
  swing1Motion=Free
  swing2LimitDegrees=45
  swing2Motion=Free
  twistLimitDegrees=45
  twistMotion=Free
 }
]
