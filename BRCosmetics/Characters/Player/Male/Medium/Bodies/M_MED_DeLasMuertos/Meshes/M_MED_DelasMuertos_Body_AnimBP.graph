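# Animation Blueprint for the male medium DeLasMuertos cosmetic body, adding secondary motion to a pose copied from the attached parent mesh.
# Weighted constraints drive shoulder pads from clavicles and upper-arm twists, then Look Ats aim front pad bones at chest points.
# Update event maps each upper arm's yaw relative to its clavicle into Look At alpha.
# AnimDynamics swings the tassel bones before a pelvis-space RigidBody node simulates the dynamics physics asset.
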AnimBlueprint /BRCosmetics/Characters/Player/Male/Medium/Bodies/M_MED_DeLasMuertos/Meshes/M_MED_DelasMuertos_Body_AnimBP
target_skeleton=&/M_MED_DelasMuertos_Body_Skeleton
graphs [
 AnimGraph {nodes [
   n1 = "Constraint - pd_shoulderPad_l (PP)"!(
    ComponentPose=Local To Component(LocalPose=Copy Pose From Mesh() {bUseAttachedParent=true})
    Alpha=1
    ConstraintWeights_0=0.2262
    ConstraintWeights_1:float=0.7738
   ) -> (Pose) {
    alphaBoolBlend {blendOption=Linear}
    boneToModify {boneName=pd_shoulderPad_l}
    constraintSetup [{targetBone {boneName=upperarm_twist_01_l}}, {targetBone {boneName=clavicle_l}}]
    constraintWeights [1,1]
   }
   n2 = "Constraint - pd_shoulderPad_r (PP)"!(ComponentPose=@n1, Alpha=1, ConstraintWeights_0=0.7738, ConstraintWeights_1:float=0.2262) -> (Pose) {
    alphaBoolBlend {blendOption=Linear}
    boneToModify {boneName=pd_shoulderPad_r}
    constraintSetup [{targetBone {boneName=clavicle_r}}, {targetBone {boneName=upperarm_twist_01_r}}]
    constraintWeights [1,1]
   }
   n3 = Look At - Bone: dyn_shoulderPad_fr_r(ComponentPose=@n2, Alpha=Get arm up r() -> (arm up r:double)) {
    alphaBoolBlend {blendOption=Linear}
    boneToModify {boneName=dyn_shoulderPad_fr_r}
    lookAtLocation=<0 4>
    lookAtTarget {boneReference {boneName=dyn_chest_aim_point_r}}
    lookAt_Axis {axis=<-1 0>}
   }
   n4 = Look At - Bone: dyn_shoulderPad_fr_l(ComponentPose=@n3, Alpha=Get Arm Up L() -> (Arm Up L:double)) {
    alphaBoolBlend {blendOption=Linear}
    boneToModify {boneName=dyn_shoulderPad_fr_l}
    lookAtLocation=<0 -4>
    lookAtTarget {boneReference {boneName=dyn_chest_aim_point_l}}
    lookAt_Axis {axis=<1 0>}
   }
   n5 = Anim Dynamics - Bone: dyn_tassle_r_01(ComponentPose=@n4, Alpha=1, ExternalForce="0,0,0") {
    angularDampingOverride=0.98
    bGravityOverrideInSimSpace=true
    bOverrideAngularDamping=true
    bUseGravityOverride=true
    simulationSpace=RootRelative
    alphaBoolBlend {blendOption=Linear}
    boundBone {boneName=dyn_tassle_r_01}
    gravityOverride {z=-980}
    physicsBodyDefinitions [{
      sphereCollisionRadius=0
      boundBone {boneName=dyn_tassle_r_01}
      boxExtents=<15 15 8>
      constraintSetup {twistAxis=AxisZ, angularLimitsMax=<0 10 5>, angularLimitsMin=<-30 -10 -5>, angularTarget=<0 0 0>}
      localJointOffset=<0 0 -3>
     }]
   }
   n6 = Anim Dynamics - Bone: dyn_tassle_r_02(ComponentPose=@n5, Alpha=1, ExternalForce="0,0,0") {
    angularDampingOverride=0.98
    bGravityOverrideInSimSpace=true
    bOverrideAngularDamping=true
    bUseGravityOverride=true
    simulationSpace=RootRelative
    alphaBoolBlend {blendOption=Linear}
    boundBone {boneName=dyn_tassle_r_02}
    gravityOverride {z=-980}
    physicsBodyDefinitions [{
      sphereCollisionRadius=0
      boundBone {boneName=dyn_tassle_r_02}
      boxExtents=<15 15 8>
      constraintSetup {twistAxis=AxisZ, angularLimitsMax=<0 10 5>, angularLimitsMin=<-30 -10 -5>, angularTarget=<0 0 0>}
      localJointOffset=<0 0 -3>
     }]
   }
   n7 = Anim Dynamics - Bone: dyn_tassle_r_03(ComponentPose=@n6, Alpha=1, ExternalForce="0,0,0") {
    angularDampingOverride=0.98
    bGravityOverrideInSimSpace=true
    bOverrideAngularDamping=true
    bUseGravityOverride=true
    simulationSpace=RootRelative
    alphaBoolBlend {blendOption=Linear}
    boundBone {boneName=dyn_tassle_r_03}
    gravityOverride {z=-980}
    physicsBodyDefinitions [{
      sphereCollisionRadius=0
      boundBone {boneName=dyn_tassle_r_03}
      boxExtents=<15 15 8>
      constraintSetup {twistAxis=AxisZ, angularLimitsMax=<0 10 5>, angularLimitsMin=<-30 -10 -5>, angularTarget=<0 0 0>}
      localJointOffset=<0 0 -3>
     }]
   }
   n8 = Anim Dynamics - Bone: dyn_tassle_l_01(ComponentPose=@n7, Alpha=1, ExternalForce="0,0,0") {
    angularDampingOverride=0.98
    bGravityOverrideInSimSpace=true
    bOverrideAngularDamping=true
    bUseGravityOverride=true
    simulationSpace=RootRelative
    alphaBoolBlend {blendOption=Linear}
    boundBone {boneName=dyn_tassle_l_01}
    gravityOverride {z=-980}
    physicsBodyDefinitions [{
      sphereCollisionRadius=0
      boundBone {boneName=dyn_tassle_l_01}
      boxExtents=<15 15 8>
      constraintSetup {twistAxis=AxisZ, angularLimitsMax=<30 10 5>, angularLimitsMin=<0 -10 -5>, angularTarget=<0 0 0>}
      localJointOffset=<0 0 -3>
     }]
   }
   n9 = Anim Dynamics - Bone: dyn_tassle_l_02(ComponentPose=@n8, Alpha=1, ExternalForce="0,0,0") {
    angularDampingOverride=0.98
    bGravityOverrideInSimSpace=true
    bOverrideAngularDamping=true
    bUseGravityOverride=true
    simulationSpace=RootRelative
    alphaBoolBlend {blendOption=Linear}
    boundBone {boneName=dyn_tassle_l_02}
    gravityOverride {z=-980}
    physicsBodyDefinitions [{
      sphereCollisionRadius=0
      boundBone {boneName=dyn_tassle_l_02}
      boxExtents=<15 15 8>
      constraintSetup {twistAxis=AxisZ, angularLimitsMax=<30 10 5>, angularLimitsMin=<0 -10 -5>, angularTarget=<0 0 0>}
      localJointOffset=<0 0 -3>
     }]
   }
   n10 = Anim Dynamics - Bone: dyn_tassle_l_03(ComponentPose=@n9, Alpha=1, ExternalForce="0,0,0") {
    angularDampingOverride=0.98
    bGravityOverrideInSimSpace=true
    bOverrideAngularDamping=true
    bUseGravityOverride=true
    simulationSpace=RootRelative
    alphaBoolBlend {blendOption=Linear}
    boundBone {boneName=dyn_tassle_l_03}
    gravityOverride {z=-980}
    physicsBodyDefinitions [{
      sphereCollisionRadius=0
      boundBone {boneName=dyn_tassle_l_03}
      boxExtents=<15 15 8>
      constraintSetup {twistAxis=AxisZ, angularLimitsMax=<30 10 5>, angularLimitsMin=<0 -10 -5>, angularTarget=<0 0 0>}
      localJointOffset=<0 0 -3>
     }]
   }
   n11 = RigidBody(ComponentPose=@n10) {
    bOverrideWorldGravity=true
    overridePhysicsAsset=&/M_MED_DelasMuertos_Dyn
    simulationSpace=BaseBoneSpace
    alphaBoolBlend {blendOption=Linear}
    baseBoneRef {boneName=pelvis}
    componentAppliedLinearAccClamp=<1000 1000 1000>
    overrideWorldGravity {z=-980}
   }
   Output Pose(Result=Component To Local(ComponentPose=@n11)) {layerGroup=None}
   Sequence Player 'RunningAround_v02'() {playRate=0, sequence=&/RunningAround_v02}
  ]}
 EventGraph {nodes [
   Event BlueprintUpdateAnimation()
   n1 = GetOwnerSkeletalMeshComponent()
   n2 = BreakTransform(InTransform=GetDeltaTransformFromRefPose(self=@n1, BoneName=upperarm_r, BaseName=clavicle_r))
   > Set arm up r(arm up r:double=MapRangeClamped(Value=BreakRotator(InRot=@n2.Rotation).Yaw, InRangeA=30, InRangeB=65)) -> (Output_Get:double)
   n3 = BreakTransform(InTransform=GetDeltaTransformFromRefPose(self=@n1, BoneName=upperarm_l, BaseName=clavicle_l))
   > Set Arm Up L(Arm Up L:double=MapRangeClamped(Value=BreakRotator(InRot=@n3.Rotation).Yaw, InRangeA=30, InRangeB=65)) -> (Output_Get:double)
   TryGetPawnOwner()
   Get bIsTargeting()
  ]}
]
variables [
 Arm Up L {type=number}
 arm up r {type=number}
]
